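; PCG subgraph that scatters points randomly over input surface areas, serving as the core process of the Fill Areas Randomly primitive.
; Surface Sampler generates seeded points from density, extents and inverted Uniformity, then per-axis lerps use SurfaceRotation to blend between surface alignment and zero rotation.
; A branch applies Random Choice when FinalCount exceeds -1.
; Transform Points applies position, rotation and scale ranges before point extents are overwritten for output.
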(graph Fill_Areas_Randomly_CoreProcess
  (params (PointsPerSquaredMeter :type Double :default 1e+01) (SamplingPointExtents :type Vector :default (1.0 1.0 1.0)) (Uniformity :type Double) (FinalCount :type Int :default -1) (SamplingSeed :type Int :default 1924953654) (OutputPointsExtents :type Vector :default (1e+01 1e+01 1e+01)) (SurfaceRotation :type Double :default 1.0) (PositionRange :type Vector) (RotationRange :type Rotator) (ScaleRange :type Vector))
  (bind AttributeMathsOp_7 (Attribute%20Maths%20Op :Operation OneMinus :In Uniformity)) ; Inverts Uniformity for the Surface Sampler's Looseness pin\n— they are opposite concepts.
  (bind FilterDataByType_0 (Filter%20Data%20By%20Type :TargetType (struct :Ids [(struct :Struct /Script/PCG.PCGDataTypeInfoSurface)])))
  (bind SurfaceSampler_1 (Surface%20Sampler :bUnbounded true :PointExtents (1.0 1.0 1.0) :PointsPerSquaredMeter 2.0 :PointExtents-param SamplingPointExtents :PointsPerSquaredMeter-param PointsPerSquaredMeter :Seed-param SamplingSeed))
  (bind CreateConstant_1 (Create%20Constant :AttributeTypes (struct :Type Integer32 :Int32Value -1))) ; Sentinel value -1. Means 'no fixed-count picking'\nin the Branch below — keeps all sampled points.
  (bind AttributeCompareOp_1 (Attribute%20Compare%20Op :Operation Greater :InA FinalCount))
  (bind AttributeMathsOp_0 (Attribute%20Maths%20Op :InputSource1 $Rotation.Roll :Operation Set))
  (bind Branch_1 (Branch)) ; Routes through RandomChoice (Output B) when\nFinalCount > -1 to pick exactly N points,\notherwise passes everything through (Output A).
  (bind RandomChoice_0 (Random%20Choice :FixedNumber-param FinalCount))
  (bind AttributeMathsOp_1 (Attribute%20Maths%20Op :Operation Multiply :InA PositionRange))
  (bind AttributeMathsOp_4 (Attribute%20Maths%20Op :InA ScaleRange))
  (bind AttributeMathsOp_3 (Attribute%20Maths%20Op :Operation Multiply :InA ScaleRange))
  (bind AttributeMathsOp_5 (Attribute%20Maths%20Op))
  (bind BreakVectorAttribute_1 (Break%20Vector%20Attribute :In RotationRange))
  (bind MakeVectorAttribute_0 (Make%20Vector%20Attribute :OutputType Vector))
  (bind AttributeMathsOp_2 (Attribute%20Maths%20Op :InputSource1 @Last :Operation Multiply)) ; @Last refers to the previous attribute in the chain.\nMultiplies RotationRange by the per-point random factor.
  (bind BreakVectorAttribute_2 (Break%20Vector%20Attribute))
  (bind MakeRotatorAttribute_0 (Make%20Rotator%20Attribute :Operation MakeRotFromAngles)) ; RotationRange vector is mapped as X→Roll, Y→Pitch, Z→Yaw.\nNot the standard UE Rotator order (Pitch, Yaw, Roll).
  (bind TransformPoints_1 (Transform%20Points :OffsetMax (1e+02 1e+02 0.0) :OffsetMin (-1e+02 -1e+02 0.0) :RotationMax (struct :Yaw 3.6e+02) :OffsetMax-param PositionRange :RotationMax-param RotationRange))
  (bind AttributeMathsOp_6 (Attribute%20Maths%20Op :InputSource1 $Extents :Operation Set :InB OutputPointsExtents)) ; Final step: overwrites each point's $Extents with\nthe user-configured OutputPointsExtents value before output.
  ;[0.15 0.15 0.15 0.5] Surface Rotation
  (bind AttributeMathsOp_9 (Attribute%20Maths%20Op :Operation OneMinus :In SurfaceRotation)) ; Used as the Lerp ratio for the rotation blend below.\nSurfaceRotation=1 keeps full surface alignment,\n0 zeros out the rotation.
  ;[0.85 0.85 0.45 0.4] Surface Rotation blend — lerps Roll, Pitch, then Yaw from the
  ; surface-aligned rotation toward zero. Ratio is driven by
  ; (1 - SurfaceRotation), so 1=keep alignment, 0=zero rotation.
  (bind AttributeMathsOp_8 (Attribute%20Maths%20Op :InputSource1 $Rotation.Roll :InputSource2 $Rotation.Roll :Operation Lerp :OutputTarget $Rotation.Roll))
  (bind AttributeMathsOp_10 (Attribute%20Maths%20Op :InputSource1 $Rotation.Pitch :InputSource2 $Rotation.Pitch :Operation Lerp :OutputTarget $Rotation.Pitch))
  (bind AttributeMathsOp_11 (Attribute%20Maths%20Op :InputSource1 $Rotation.Yaw :InputSource2 $Rotation.Yaw :Operation Lerp :OutputTarget $Rotation.Yaw))
  (connect CreateConstant_1 Out AttributeCompareOp_1 InB)
  (connect AttributeMathsOp_11 Out AttributeMathsOp_0 InA)
  (connect MakeVectorAttribute_0 Out AttributeMathsOp_2 InA)
  (connect AttributeMathsOp_3 Out AttributeMathsOp_5 InA)
  (connect TransformPoints_1 Out AttributeMathsOp_6 InA)
  (connect SurfaceSampler_1 Out AttributeMathsOp_8 InA)
  (connect AttributeMathsOp_9 Out AttributeMathsOp_8 Ratio)
  (connect AttributeMathsOp_8 Out AttributeMathsOp_10 InA)
  (connect AttributeMathsOp_9 Out AttributeMathsOp_10 Ratio)
  (connect AttributeMathsOp_10 Out AttributeMathsOp_11 InA)
  (connect AttributeMathsOp_9 Out AttributeMathsOp_11 Ratio)
  (connect AttributeCompareOp_1 Out Branch_1 bOutputToB)
  (connect AttributeMathsOp_0 Out Branch_1 In)
  (connect AttributeMathsOp_2 Out BreakVectorAttribute_2 In)
  (connect DefaultInputNode In FilterDataByType_0 In)
  (connect BreakVectorAttribute_2 Y MakeRotatorAttribute_0 Pitch)
  (connect BreakVectorAttribute_2 X MakeRotatorAttribute_0 Roll)
  (connect BreakVectorAttribute_2 Z MakeRotatorAttribute_0 Yaw)
  (connect BreakVectorAttribute_1 X MakeVectorAttribute_0 X)
  (connect BreakVectorAttribute_1 Y MakeVectorAttribute_0 Y)
  (connect BreakVectorAttribute_1 Z MakeVectorAttribute_0 Z)
  (connect Branch_1 Output%20B RandomChoice_0 In)
  (connect AttributeMathsOp_7 Out SurfaceSampler_1 Looseness)
  (connect FilterDataByType_0 InsideFilter SurfaceSampler_1 Surface)
  (connect Branch_1 Output%20A TransformPoints_1 In)
  (connect RandomChoice_0 Chosen TransformPoints_1 In)
  (connect AttributeMathsOp_1 Out TransformPoints_1 OffsetMin)
  (connect MakeRotatorAttribute_0 Out TransformPoints_1 RotationMin)
  (connect AttributeMathsOp_4 Out TransformPoints_1 ScaleMax)
  (connect AttributeMathsOp_5 Out TransformPoints_1 ScaleMin)
  (output Out AttributeMathsOp_6))
