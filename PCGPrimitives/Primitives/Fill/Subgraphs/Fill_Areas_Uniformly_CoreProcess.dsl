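; PCG core subgraph that uniformly fills 2D polygon areas with points aligned to each polygon's longest outer edge.
; It moves control points into that edge's local space, surface-samples a closed linear spline, then rotates points back to world space.
; Sampler extents come from the Asset's 'Root' data bounds or mesh bounds, or from PointExtents when Asset is unset.
; A set Asset is also stamped onto every output point.
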(graph Fill_Areas_Uniformly_CoreProcess
  (params (PointExtents :type Double :default 1e+02) (Asset :type (SoftObject /Script/CoreUObject.Object)))
  (bind FilterDataByType_1 (Filter%20Data%20By%20Type :TargetType (struct :Ids [(struct :Struct /Script/PCG.PCGDataTypeInfoPolygon2D)])))
  (bind GetSplineControlPoints_0 (Get%20Spline%20Control%20Points))
  (bind BreakTransformAttribute_1 (Break%20Transform%20Attribute))
  (bind AttributeCast_2 (Attribute%20Cast :OutputType Rotator))
  (bind AttributeMathsOp_1 (Attribute%20Maths%20Op :InputSource1 $Rotation.Roll :Operation Set))
  ;[0.85 0.55 0.4 0.4] Capture the polygon's longest outer edge as a Transform attribute set.
  ; $LongestOuterSegmentIndex picks the segment, then we filter the spline
  ; control points to that index and extract its Transform. The result is
  ; used below to sample the polygon in its own local space.
  (bind CopyAttributes_1 (Copy%20Attributes :InputSource @Data.$LongestOuterSegmentIndex :OutputTarget longestSeg))
  (bind CopyAttributes_2 (Copy%20Attributes :InputSource $SegmentLength :OutputTarget lenght))
  (bind CopyAttributes_3 (Copy%20Attributes :InputSource $LocalRotation :OutputTarget localRot))
  (bind AttributeCast_1 (Attribute%20Cast :OutputType Rotator))
  (bind GetSplineControlPoints_1 (Get%20Spline%20Control%20Points))
  (bind AttributeFilter_1 (Filter%20Attribute%20Elements :Operator Equal :TargetAttribute $Index :ThresholdAttribute longestSeg)) ; Filter spline points to keep only the one\nat $Index == longestSeg — the start of\nthe polygon's longest outer edge.
  (bind CopyAttributes_4 (Copy%20Attributes :InputSource $Transform :OutputTarget PointTransform))
  (bind PointToAttributeSet_0 (Point%20To%20Attribute%20Set))
  (bind DeleteAttributes_0 (Delete%20Attributes :Operation KeepSelectedAttributes :SelectedAttributes "PointTransform"))
  ;[0.85 0.85 0.45 0.4] Sample the polygon uniformly in the longest-edge's local space, then
  ; return to world space. This rotates the sampling grid to align with
  ; the polygon's longest edge — without this step the axis-aligned grid
  ; would not match the polygon's shape.
  ;;[0.15 0.15 0.15 0.5] APPLY TRANSFORM
  (bind AttributeVectorOp_1 (Attribute%20Vector%20Op :InputSource1 $Position :Operation InverseTransformDirection))
  ;;[0.15 0.15 0.15 0.5] FILL
  (bind CreateSpline_0 (Create%20Spline :bClosedLoop true :bLinear true :InterpTypeAttribute None))
  (bind CreateSurfaceFromSpline_0 (Create%20Surface%20From%20Spline))
  (bind SurfaceSampler_0 (Surface%20Sampler :bUnbounded true :Looseness 0.0 :PointsPerSquaredMeter 1e+02))
  ;;[0.15 0.15 0.15 0.5] REVERT TRANSFORM
  (bind AttributeVectorOp_0 (Attribute%20Vector%20Op :InputSource1 $Position :Operation TransformDirection))
  (bind CopyAttributes_5 (Copy%20Attributes :InputSource PointTransform :OutputTarget $Rotation))
  ;[0.15 0.15 0.15 0.5] Case PCG assemblies
  ;;[0.45 0.65 0.85 0.4] Pick PointExtents for the Surface Sampler. Priority chain:
  ;; 1) If Asset param is unset, use the Point Extents graph parameter.
  ;; 2) Else if Asset is a PCG data asset, use its 'Root' tagged bounds.
  ;; 3) Else fall back to Bounds From Mesh on the asset.
  (bind CreateConstant_0 (Create%20Constant :AttributeTypes (struct :Type String))) ; Empty-string sentinel: comparison below detects\n'Asset is unset' to skip the load path.
  (bind AttributeCompareOp_1 (Attribute%20Compare%20Op :InA Asset))
  (bind Branch_0 (Branch :In Asset))
  (bind PCGLoadDataAsset_2 (Load%20PCG%20Data%20Asset :bLoadFromInput true))
  (bind FilterDataByTag_2 (Filter%20Data%20By%20Tag :SelectedTags "Root")) ; PCG asset convention: 'Root' tagged data is the\nentry-level entry — filter to it before reading bounds.
  (bind GetBounds_0 (Get%20Bounds))
  (bind AttributeRename_1 (Attribute%20Rename :AttributeToRename BoundsMax :NewAttributeName PointExtents)) ; Rename BoundsMax → PointExtents so the downstream\nSelect reads it under the expected attribute name.
  (bind DataNum_0 (Data%20Count))
  (bind AttributeSetToPoint_0 (Attribute%20Set%20To%20Point :In Asset))
  (bind BoundsFromMesh_0 (Bounds%20From%20Mesh))
  (bind ExtractAttribute_0 (Extract%20Attribute%20at%20Index :bForceOutputAttributeToBeInElementsDomain true :InputSource $Extents :OutputAttributeName Point_Extent))
  (bind Select_1 (Select))
  (bind Select_0 (Select :Input%20B PointExtents))
  ;[0.45 0.85 0.55 0.4] If the Asset parameter is set, stamp it as an Asset attribute
  ; on every output point. Otherwise pass through unmodified.
  (bind CreateConstant_1 (Create%20Constant :AttributeTypes (struct :Type String))) ; Same empty-string sentinel as the other Create Constant —\nBranch_2 uses it to skip stamping the Asset attribute.
  (bind AttributeCompareOp_2 (Attribute%20Compare%20Op :InA Asset))
  (bind Branch_2 (Branch))
  (bind AddAttribute_1 (Add%20Attribute :OutputTarget Asset :Attributes Asset))
  (connect Branch_2 Output%20A AddAttribute_1 In)
  (connect CopyAttributes_3 Out AttributeCast_1 In)
  (connect BreakTransformAttribute_1 Rotation AttributeCast_2 In)
  (connect CreateConstant_0 Out AttributeCompareOp_1 InB)
  (connect CreateConstant_1 Out AttributeCompareOp_2 InB)
  (connect GetSplineControlPoints_1 Out AttributeFilter_1 In)
  (connect CopyAttributes_5 Out AttributeMathsOp_1 InA)
  (connect GetBounds_0 Out AttributeRename_1 In)
  (connect SurfaceSampler_0 Out AttributeVectorOp_0 In)
  (connect DeleteAttributes_0 Out AttributeVectorOp_0 Transform)
  (connect GetSplineControlPoints_0 Out AttributeVectorOp_1 In)
  (connect DeleteAttributes_0 Out AttributeVectorOp_1 Transform)
  (connect AttributeSetToPoint_0 Out BoundsFromMesh_0 In)
  (connect AttributeCompareOp_1 Out Branch_0 bOutputToB)
  (connect AttributeCompareOp_2 Out Branch_2 bOutputToB)
  (connect AttributeMathsOp_1 Out Branch_2 In)
  (connect DeleteAttributes_0 Out BreakTransformAttribute_1 In)
  (connect FilterDataByType_1 InsideFilter CopyAttributes_1 Source)
  (connect FilterDataByType_1 InsideFilter CopyAttributes_1 Target)
  (connect CopyAttributes_1 Out CopyAttributes_2 Source)
  (connect CopyAttributes_1 Out CopyAttributes_2 Target)
  (connect CopyAttributes_2 Out CopyAttributes_3 Source)
  (connect CopyAttributes_2 Out CopyAttributes_3 Target)
  (connect AttributeFilter_1 InsideFilter CopyAttributes_4 Source)
  (connect AttributeFilter_1 InsideFilter CopyAttributes_4 Target)
  (connect AttributeCast_2 Out CopyAttributes_5 Source)
  (connect AttributeVectorOp_0 Out CopyAttributes_5 Target)
  (connect AttributeVectorOp_1 Out CreateSpline_0 In)
  (connect CreateSpline_0 Out CreateSurfaceFromSpline_0 In)
  (connect PCGLoadDataAsset_2 Out DataNum_0 In)
  (connect PointToAttributeSet_0 Out DeleteAttributes_0 In)
  (connect BoundsFromMesh_0 Out ExtractAttribute_0 In)
  (connect PCGLoadDataAsset_2 Out FilterDataByTag_2 In)
  (connect DefaultInputNode In FilterDataByType_1 In)
  (connect FilterDataByTag_2 InsideFilter GetBounds_0 In)
  (connect FilterDataByType_1 InsideFilter GetSplineControlPoints_0 In)
  (connect AttributeCast_1 Out GetSplineControlPoints_1 In)
  (connect Branch_0 Output%20A PCGLoadDataAsset_2 In)
  (connect CopyAttributes_4 Out PointToAttributeSet_0 In)
  (connect AttributeCompareOp_1 Out Select_0 bUseInputB)
  (connect Select_1 Out Select_0 Input%20A)
  (connect DataNum_0 Out Select_1 bUseInputB)
  (connect ExtractAttribute_0 Out Select_1 Input%20A)
  (connect AttributeRename_1 Out Select_1 Input%20B)
  (connect Select_0 Out SurfaceSampler_0 PointExtents)
  (connect CreateSurfaceFromSpline_0 Out SurfaceSampler_0 Surface)
  (output Out AddAttribute_1)
  (output Out Branch_2.Output%20B))
