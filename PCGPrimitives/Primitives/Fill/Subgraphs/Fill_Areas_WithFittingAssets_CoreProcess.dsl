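; PCG subgraph that fills polygon areas with randomly assigned, non-overlapping meshes or PCG data asset assemblies, spawning them or only outputting points.
; It samples a grid aligned to each polygon's longest edge, assigns assets by seed, and derives bounds from mesh or assembly root.
; Back in world space, points optionally get 90° yaw on half, lose border overlaps, then self-prune by scaled extents.
; Spawning unrolls assemblies and chunks points for performance.
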(graph Fill_Areas_WithFittingAssets_CoreProcess
  (params (Seed :type Int) (NewProperty :type Int) (overlap :type Double :default 1.0) (Spawn_Assets :type Bool :default true) (Keep_Assemblies_as_assets :type Bool) (random_90_rotation :type Bool) (Base_fill_points_per_square_meters :type Double :default 1.0) (Base_fill_point_extents :type Double :default 1e+02))
  (bind SplineSampler_0 (Spline%20Sampler :SamplerParams (struct :bUnbounded true :PointSteepness 1.0)))
  (bind BoundsModifier_0 (Bounds%20Modifier :BoundsMax (2.0 1e+02 8e+02) :BoundsMin (0.0 1e+02 1e+02))) ; Turn border samples into thin exclusion volumes:\nY=±100cm wide, Z=100-800cm tall. Wide enough to overlap\nany nearby interior point so Difference removes them.
  (bind FilterDataByType_1 (Filter%20Data%20By%20Type :TargetType (struct :Ids [(struct :Struct /Script/PCG.PCGDataTypeInfoPolygon2D)])))
  (bind BreakTransformAttribute_1 (Break%20Transform%20Attribute))
  (bind AttributeCast_2 (Attribute%20Cast :OutputType Rotator))
  (bind WriteDataIndex_0 (Write%20Data%20Index :bAddAttribute true :bAddTag false :IndexAttribute dataIndex))
  (bind Branch_2 (Branch :bOutputToB-param random_90_rotation))
  (bind Difference_0 (Difference :DensityFunction Binary)) ; removing those which overlap border
  (bind ToPoint_0 (To%20Point))
  (bind MergePoints_0 (Merge%20Points :MergeTagsMode None))
  (bind Branch_0 (Branch :bOutputToB-param Spawn_Assets)) ; Route by Spawn Assets parameter. Output A goes to the\npass-through path (produce points only). Output B feeds\nthe full spawn pipeline with PCG-asset loading and meshes.
  ;[0.95 0.45 0.45 0.4] Reject inputs that would break downstream:
  ; polygon area below 10, or empty Assets list.
  ; Also lifts polygon Z by MaxHeight from the asset list
  ; so downstream bounds extend over the tallest asset.
  (bind FilterDataByAttribute_0 (Filter%20Data%20By%20Attribute :FilterMode FilterByValue :TargetAttribute @Data.$Area :Threshold (struct :AttributeTypes (struct :DoubleValue 1e+01)))) ; safe check  for area too small that would break down stream
  ;;[0.15 0.15 0.15 0.5] IF no Assets dont start
  (bind GetElementsCount_1 (Get%20Element%20Count :In Assets))
  (bind Branch_3 (Branch))
  ;[0.85 0.55 0.4 0.4] Capture the polygon's longest outer edge as a Transform attribute set.
  ; $LongestOuterSegmentIndex picks the segment, then we filter spline
  ; control points to that index and extract its Transform.
  ; Used below to sample in the polygon's local space.
  (bind CopyAttributes_1 (Copy%20Attributes :InputSource @Data.$LongestOuterSegmentIndex :OutputTarget longestSeg))
  (bind CopyAttributes_2 (Copy%20Attributes :InputSource $SegmentLength :OutputTarget lenght))
  (bind CopyAttributes_3 (Copy%20Attributes :InputSource $LocalRotation :OutputTarget localRot))
  (bind AttributeCast_1 (Attribute%20Cast :OutputType Rotator))
  (bind GetSplineControlPoints_2 (Get%20Spline%20Control%20Points))
  (bind AttributeFilter_2 (Filter%20Attribute%20Elements :Operator Equal :TargetAttribute $Index :ThresholdAttribute longestSeg))
  (bind CopyAttributes_4 (Copy%20Attributes :InputSource $Transform :OutputTarget PointTransform))
  (bind PointToAttributeSet_0 (Point%20To%20Attribute%20Set))
  (bind DeleteAttributes_0 (Delete%20Attributes :Operation KeepSelectedAttributes :SelectedAttributes "PointTransform"))
  ;[0.85 0.85 0.45 0.4] Sample the polygon uniformly in the longest-edge's local space.
  ; This rotates the sampling grid to align with the polygon's
  ; longest edge — without it the axis-aligned grid would not
  ; match the polygon's shape. Brought back to world space later.
  (bind GetSplineControlPoints_0 (Get%20Spline%20Control%20Points))
  (bind CreateSpline_0 (Create%20Spline :bClosedLoop true :bLinear true :InterpTypeAttribute None))
  (bind CreateSurfaceFromSpline_0 (Create%20Surface%20From%20Spline))
  (bind SurfaceSampler_0 (Surface%20Sampler :bUnbounded true :Looseness 0.0 :PointExtents (1e+02 1e+02 1e+02) :PointsPerSquaredMeter 1.0 :PointExtents-param Base_fill_point_extents :PointsPerSquaredMeter-param Base_fill_points_per_square_meters))
  ;;[0.15 0.15 0.15 0.5] APPLY TRANSFORM
  (bind AttributeVectorOp_1 (Attribute%20Vector%20Op :InputSource1 $Position :Operation InverseTransformDirection))
  ;[0.15 0.15 0.15 0.5] SAFE CHECK for empty data, as sometime the surface sampler,
  ; Fails to return points.
  (bind GetElementsCount_0 (Get%20Element%20Count))
  (bind CopyAttributes_7 (Copy%20Attributes :InputSource $Index :OutputTarget indexS))
  (bind AttributeFilter_4 (Filter%20Attribute%20Elements :AttributeTypes (struct :DoubleValue 1e+01) :bUseConstantThreshold true :TargetAttribute NumEntries :ThresholdAttribute @Last))
  (bind AttributeReduce_1 (Attribute%20Reduce :InputSource indexS :Operation Join))
  (bind FilterDataByIndex_2 (Filter%20Data%20By%20Index))
  (bind FilterDataByIndex_1 (Filter%20Data%20By%20Index))
  ;[0.15 0.15 0.15 0.5] CREATE POINTS and ASSIGN ASSETS
  (bind AttributePartition_3 (Attribute%20Partition :PartitionAttributeSelectors [dataIndex]))
  ;;[0.45 0.85 0.85 0.4] Randomly assign each sampled point one entry from the Assets list,
  ;; stamp that entry's attributes onto it, extract its class to drive
  ;; per-type routing, then group and process each asset class in
  ;; the Loop below (handles PCGDataAsset vs static-mesh paths).
  (bind WriteDataIndex_1 (Write%20Data%20Index :bAddAttribute true :bAddTag false :IndexAttribute dataIndex))
  (bind MatchAndSetAttributes_0 (Match%20And%20Set%20Attributes :Match%20Data Assets :Seed-param Seed)) ; Random asset assignment: picks one entry from the Assets\nlist per point, stamping that entry's attributes onto it.\nSeeded by the Seed graph parameter for reproducibility.
  (bind GetClassFromAttribute_1 (Get%20Class%20From%20Attribute :InputSource Asset :OutputTarget "Asset Type Str"))
  (bind MergePoints_3 (Merge%20Points :MergeTagsMode None))
  (bind AttributePartition_5 (Attribute%20Partition :PartitionAttributeSelectors ["Asset Type Str"]))
  ;;[0.15 0.15 0.15 0.5] Assemblies or StaticMeshes
  (bind Loop_0 (Loop :Graph Asset_Type_LOOP_1))
  ;[0.15 0.15 0.15 0.5] REVERT TRANSFORM
  (bind AttributeVectorOp_0 (Attribute%20Vector%20Op :InputSource1 $Position :Operation TransformDirection))
  (bind CopyAttributes_5 (Copy%20Attributes :InputSource PointTransform :OutputTarget $Rotation))
  (bind AttributeMathsOp_2 (Attribute%20Maths%20Op :InputSource1 $Rotation.Roll :Operation Set))
  ;[0.15 0.15 0.15 0.5] 50% of them are rotated
  (bind AttributeNoise_1 (Attribute%20Noise))
  (bind AttributeFilter_1 (Filter%20Attribute%20Elements :AttributeTypes (struct :DoubleValue 0.5) :bUseConstantThreshold true :ThresholdAttribute @Last)) ; Threshold 0.5 on the noise above randomly splits points\nin half: InsideFilter keeps original yaw, OutsideFilter\ngoes through the 90° yaw rotation branch.
  (bind TransformPoints_1 (Transform%20Points :RotationMax (struct :Yaw 9e+01) :RotationMin (struct :Yaw 9e+01)))
  (bind MergePoints_1 (Merge%20Points :MergeTagsMode None))
  ;[0.95 0.6 0.3 0.4] The 'fitting' core: scale extents by Overlap, then
  ; Self Pruning removes overlapping points so no two assets
  ; clash in the output. Finally partition by Asset so each
  ; asset's points can route through the spawn paths below.
  (bind ExtentsModifier_0 (Extents%20Modifier :Mode Multiply :Extents-param overlap)) ; Scale extents by the Overlap graph parameter before\nself-pruning. Overlap < 1 packs tighter, > 1 looser.\nLets users tune asset density without re-sampling.
  (bind SelfPruning_1 (Self%20Pruning :Parameters (struct :PruningType AllEqual :ComparisonSource $Extents :RadiusSimilarityFactor 1.0))) ; The 'fitting' guarantee: removes any points whose $Extents\nwould overlap. Randomized so it's not just first-wins —\nthis is what ensures no two final assets clash.
  (bind AttributePartition_1 (Attribute%20Partition :PartitionAttributeSelectors [Asset]))
  ;[0.15 0.15 0.15 0.5] Assemblies
  (bind GetAttributeFromPointIndex_1 (Get%20Attribute%20From%20Point%20Index :bForceOutputAttributeToBeInElementsDomain true :InputSource Asset :OutputAttributeName None))
  (bind PCGLoadDataAsset_0 (Load%20PCG%20Data%20Asset :bLoadFromInput true))
  (bind FilterDataByTag_2 (Filter%20Data%20By%20Tag :SelectedTags "Points"))
  (bind ToPoint_3 (To%20Point))
  (bind CopyPoints_1 (Copy%20Points :bCopyEachSourceOnEveryTarget false :ScaleInheritance Source))
  (bind AttributeRename_0 (Attribute%20Rename :AttributeToRename Mesh :NewAttributeName Asset)) ; Rename 'Mesh' → 'Asset' so downstream nodes read\nthe asset reference under the same name used\nthroughout the graph.
  ;[0.15 0.15 0.15 0.5] Chunking points to ease on performance
  (bind AddAttribute_0 (Add%20Attribute :AttributeTypes (struct :Type Integer32) :OutputTarget ChunkIndex))
  (bind MergePoints_2 (Merge%20Points :MergeTagsMode None))
  (bind AttributeMathsOp_4 (Attribute%20Maths%20Op :InputSource1 $Index :Operation Divide :OutputTarget chunkindex))
  (bind AttributeMathsOp_3 (Attribute%20Maths%20Op :InputSource1 chunkindex :Operation Floor))
  (bind AttributePartition_2 (Attribute%20Partition :PartitionAttributeSelectors [ChunkIndex]))
  (bind StaticMeshSpawner_0 (Static%20Mesh%20Spawner :MeshSelectorParameters (class:/Script/PCG.PCGMeshSelectorByAttribute :AttributeName Asset) :MeshSelectorType /Script/PCG.PCGMeshSelectorByAttribute))
  ;[0.4 0.75 0.65 0.4] Spawn Assets = OFF path: pass through without spawning.
  ; PCGDataAsset entries are unrolled into their constituent
  ; Points by default, or kept as single 'assembly' Asset refs
  ; if Keep Assemblies As Assets is on. Both paths merge here.
  (bind AttributeFilter_0 (Filter%20Attribute%20Elements :AttributeTypes (struct :Type String :StringValue "PCGDataAsset") :bUseConstantThreshold true :Operator Equal :TargetAttribute "Asset Type Str" :ThresholdAttribute @Last))
  (bind GetAttributeFromPointIndex_4 (Get%20Attribute%20From%20Point%20Index :bForceOutputAttributeToBeInElementsDomain true :InputSource Asset :OutputAttributeName None))
  (bind PCGLoadDataAsset_4 (Load%20PCG%20Data%20Asset :bLoadFromInput true))
  (bind FilterDataByTag_4 (Filter%20Data%20By%20Tag :SelectedTags "Points"))
  (bind ToPoint_5 (To%20Point))
  (bind CopyPoints_2 (Copy%20Points :bCopyEachSourceOnEveryTarget false :ScaleInheritance Source))
  (bind AttributeRename_2 (Attribute%20Rename :AttributeToRename Mesh :NewAttributeName Asset))
  (bind ToPoint_6 (To%20Point))
  (bind CopyPoints_3 (Copy%20Points :bCopyEachSourceOnEveryTarget false :ScaleInheritance Source))
  (bind Select_0 (Select :bUseInputB-param Keep_Assemblies_as_assets))
  (bind ToPoint_2 (To%20Point))
  (bind AttributeRename_1 (Attribute%20Rename :AttributeToRename Mesh :NewAttributeName Asset))
  (bind MergePoints_4 (Merge%20Points :MergeTagsMode None))
  (bind AttributePartition_4 (Attribute%20Partition :PartitionAttributeSelectors [dataIndex]))
  (connect AttributeRename_0 Out AddAttribute_0 In)
  (connect Branch_0 Output%20B AddAttribute_0 In)
  (connect CopyAttributes_3 Out AttributeCast_1 In)
  (connect BreakTransformAttribute_1 Rotation AttributeCast_2 In)
  (connect Branch_0 Output%20A AttributeFilter_0 In)
  (connect AttributeNoise_1 Out AttributeFilter_1 In)
  (connect GetSplineControlPoints_2 Out AttributeFilter_2 In)
  (connect CopyAttributes_7 Out AttributeFilter_4 In)
  (connect CopyAttributes_5 Out AttributeMathsOp_2 InA)
  (connect AttributeMathsOp_4 Out AttributeMathsOp_3 In)
  (connect MergePoints_2 Out AttributeMathsOp_4 InA)
  (connect Branch_2 Output%20B AttributeNoise_1 In)
  (connect SelfPruning_1 Out AttributePartition_1 In)
  (connect AttributeMathsOp_3 Out AttributePartition_2 In)
  (connect Loop_0 Out AttributePartition_3 In)
  (connect MergePoints_4 Out AttributePartition_4 In)
  (connect MergePoints_3 Out AttributePartition_5 In)
  (connect AttributeFilter_4 InsideFilter AttributeReduce_1 In)
  (connect CopyPoints_1 Out AttributeRename_0 In)
  (connect AttributeFilter_0 OutsideFilter AttributeRename_1 In)
  (connect StaticMeshSpawner_0 Out AttributeRename_1 In)
  (connect ToPoint_2 Out AttributeRename_1 In)
  (connect CopyPoints_2 Out AttributeRename_2 In)
  (connect AttributePartition_3 Out AttributeVectorOp_0 In)
  (connect FilterDataByIndex_2 InsideFilter AttributeVectorOp_0 Transform)
  (connect GetSplineControlPoints_0 Out AttributeVectorOp_1 In)
  (connect DeleteAttributes_0 Out AttributeVectorOp_1 Transform)
  (connect SplineSampler_0 Out BoundsModifier_0 In)
  (connect AttributePartition_1 Out Branch_0 In)
  (connect WriteDataIndex_0 Out Branch_2 In)
  (connect GetElementsCount_1 Out Branch_3 bOutputToB)
  (connect FilterDataByAttribute_0 InsideFilter Branch_3 In)
  (connect FilterDataByIndex_2 InsideFilter BreakTransformAttribute_1 In)
  (connect FilterDataByType_1 InsideFilter CopyAttributes_1 Source)
  (connect FilterDataByType_1 InsideFilter CopyAttributes_1 Target)
  (connect CopyAttributes_1 Out CopyAttributes_2 Source)
  (connect CopyAttributes_1 Out CopyAttributes_2 Target)
  (connect CopyAttributes_2 Out CopyAttributes_3 Source)
  (connect CopyAttributes_2 Out CopyAttributes_3 Target)
  (connect AttributeFilter_2 InsideFilter CopyAttributes_4 Source)
  (connect AttributeFilter_2 InsideFilter CopyAttributes_4 Target)
  (connect AttributeCast_2 Out CopyAttributes_5 Source)
  (connect AttributeVectorOp_0 Out CopyAttributes_5 Target)
  (connect GetElementsCount_0 Out CopyAttributes_7 Source)
  (connect GetElementsCount_0 Out CopyAttributes_7 Target)
  (connect ToPoint_3 Out CopyPoints_1 Source)
  (connect Branch_0 Output%20B CopyPoints_1 Target)
  (connect ToPoint_5 Out CopyPoints_2 Source)
  (connect AttributeFilter_0 InsideFilter CopyPoints_2 Target)
  (connect ToPoint_6 Out CopyPoints_3 Source)
  (connect AttributeFilter_0 InsideFilter CopyPoints_3 Target)
  (connect AttributeVectorOp_1 Out CreateSpline_0 In)
  (connect CreateSpline_0 Out CreateSurfaceFromSpline_0 In)
  (connect PointToAttributeSet_0 Out DeleteAttributes_0 In)
  (connect BoundsModifier_0 Out Difference_0 Differences)
  (connect Branch_2 Output%20A Difference_0 Source)
  (connect MergePoints_1 Out Difference_0 Source)
  (connect MergePoints_0 Out ExtentsModifier_0 In)
  (connect DefaultInputNode In FilterDataByAttribute_0 In)
  (connect SurfaceSampler_0 Out FilterDataByIndex_1 In)
  (connect AttributeReduce_1 Out FilterDataByIndex_1 SelectedIndices)
  (connect DeleteAttributes_0 Out FilterDataByIndex_2 In)
  (connect AttributeReduce_1 Out FilterDataByIndex_2 SelectedIndices)
  (connect PCGLoadDataAsset_0 Out FilterDataByTag_2 In)
  (connect PCGLoadDataAsset_4 Out FilterDataByTag_4 In)
  (connect Branch_3 Output%20B FilterDataByType_1 In)
  (connect Branch_0 Output%20B GetAttributeFromPointIndex_1 In)
  (connect AttributeFilter_0 InsideFilter GetAttributeFromPointIndex_4 In)
  (connect MatchAndSetAttributes_0 Out GetClassFromAttribute_1 In)
  (connect SurfaceSampler_0 Out GetElementsCount_0 In)
  (connect FilterDataByType_1 InsideFilter GetSplineControlPoints_0 In)
  (connect AttributeCast_1 Out GetSplineControlPoints_2 In)
  (connect AttributePartition_5 Out Loop_0 In)
  (connect WriteDataIndex_1 Out MatchAndSetAttributes_0 In)
  (connect ToPoint_0 Out MergePoints_0 In)
  (connect AttributeFilter_1 InsideFilter MergePoints_1 In)
  (connect TransformPoints_1 Out MergePoints_1 In)
  (connect AddAttribute_0 Out MergePoints_2 In)
  (connect GetClassFromAttribute_1 Out MergePoints_3 In)
  (connect AttributeRename_1 Out MergePoints_4 In)
  (connect GetAttributeFromPointIndex_1 Attribute PCGLoadDataAsset_0 In)
  (connect GetAttributeFromPointIndex_4 Attribute PCGLoadDataAsset_4 In)
  (connect CopyAttributes_4 Out PointToAttributeSet_0 In)
  (connect AttributeRename_2 Out Select_0 Input%20A)
  (connect CopyPoints_3 Out Select_0 Input%20B)
  (connect ExtentsModifier_0 Out SelfPruning_1 In)
  (connect Branch_3 Output%20B SplineSampler_0 Spline)
  (connect AttributePartition_2 Out StaticMeshSpawner_0 In)
  (connect CreateSurfaceFromSpline_0 Out SurfaceSampler_0 Surface)
  (connect Difference_0 Out ToPoint_0 In)
  (connect Select_0 Out ToPoint_2 In)
  (connect FilterDataByTag_2 InsideFilter ToPoint_3 In)
  (connect FilterDataByTag_4 InsideFilter ToPoint_5 In)
  (connect FilterDataByTag_4 OutsideFilter ToPoint_6 In)
  (connect AttributeFilter_1 OutsideFilter TransformPoints_1 In)
  (connect AttributeMathsOp_2 Out WriteDataIndex_0 In)
  (connect FilterDataByIndex_1 InsideFilter WriteDataIndex_1 In)
  (output Out AttributePartition_4))

(subgraph Asset_Type_LOOP_1
  (bind ExtractAttribute_0 (Extract%20Attribute%20at%20Index :bForceOutputAttributeToBeInElementsDomain true :InputSource "Asset Type Str"))
  (bind CreateConstant_0 (Create%20Constant :AttributeTypes (struct :Type String :StringValue "PCGDataAsset")))
  (bind AttributeCompareOp_0 (Attribute%20Compare%20Op))
  (bind Branch_0 (Branch))
  (bind AttributePartition_1 (Attribute%20Partition :PartitionAttributeSelectors [Asset]))
  (bind GetAttributeFromPointIndex_1 (Get%20Attribute%20From%20Point%20Index :bForceOutputAttributeToBeInElementsDomain true :InputSource Asset :OutputAttributeName None))
  (bind PCGLoadDataAsset_0 (Load%20PCG%20Data%20Asset :bLoadFromInput true))
  (bind FilterDataByTag_0 (Filter%20Data%20By%20Tag :SelectedTags "Root"))
  (bind ToPoint_0 (To%20Point))
  (bind CopyPoints_1 (Copy%20Points :bCopyEachSourceOnEveryTarget false :ScaleInheritance Source))
  (bind BoundsFromMesh_0 (Bounds%20From%20Mesh :MeshAttribute asset))
  (bind MergePoints_0 (Merge%20Points :MergeTagsMode None))
  (connect ExtractAttribute_0 Out AttributeCompareOp_0 InA)
  (connect CreateConstant_0 Out AttributeCompareOp_0 InB)
  (connect Branch_0 Output%20B AttributePartition_1 In)
  (connect Branch_0 Output%20A BoundsFromMesh_0 In)
  (connect AttributeCompareOp_0 Out Branch_0 bOutputToB)
  (connect DefaultInputNode In Branch_0 In)
  (connect ToPoint_0 Out CopyPoints_1 Source)
  (connect AttributePartition_1 Out CopyPoints_1 Target)
  (connect DefaultInputNode In ExtractAttribute_0 In)
  (connect PCGLoadDataAsset_0 Out FilterDataByTag_0 In)
  (connect AttributePartition_1 Out GetAttributeFromPointIndex_1 In)
  (connect BoundsFromMesh_0 Out MergePoints_0 In)
  (connect CopyPoints_1 Out MergePoints_0 In)
  (connect GetAttributeFromPointIndex_1 Attribute PCGLoadDataAsset_0 In)
  (connect FilterDataByTag_0 InsideFilter ToPoint_0 In)
  (output Out MergePoints_0))
